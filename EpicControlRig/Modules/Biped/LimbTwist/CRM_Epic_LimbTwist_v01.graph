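# Modular Control Rig module that spreads twist along a biped limb's twist bones, set up from Start, Mid and End connectors.
# Construction finds upper and lower twist bones by name search and optionally builds a control with parent null for each, auto-scaled to chain length.
# Forward solve blends weighted twist from Start and End onto those nulls, copying controls to bones, or onto bones directly without controls.
# Backward solve restores only null translations from the bones.
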ControlRigBlueprint /EpicControlRig/Modules/Biped/LimbTwist/CRM_Epic_LimbTwist_v01
graphs [
 ConstructionGraph {nodes [
   RigUnit_PrepareForExecution = &RigUnit_PrepareForExecution()
   RigVMFunction_Sequence_1 = RigVMAggregateNode(ExecuteContext=@RigUnit_PrepareForExecution.ExecutePin)
   RigVMFunction_ControlFlowBranch = &RigVMFunction_ControlFlowBranch(ExecuteContext=@RigVMFunction_Sequence_1.C, Condition=RigVMVariableNode() {variable=Twist Control}.Value)
   RigVMFunction_Sequence = RigVMAggregateNode(ExecuteContext=@RigVMFunction_ControlFlowBranch.True)
   RigVMDispatchNode(Array=RigVMVariableNode() {variable=Upper Twist Controls}.Value, ExecuteContext=@RigVMFunction_Sequence.A) {template="DISPATCH_RigVMDispatch_ArrayReset(io Array)"}
   > For_Each_2 = RigVMDispatchNode(Array=RigVMVariableNode() {variable=Upper Twist Bones}.Value) {template="DISPATCH_RigVMDispatch_ArrayIterator(in Array,out Element,out Index,out Count,out Ratio)"}
   Add = /Script/RigVM.RigVMFunction_MathIntAdd(B=1, A=@For_Each_2.Index) {is=struct RigVMUnitNode, template="Add::Execute(in A,in B,out Result)"}
   Int to Name = /Script/RigVM.RigVMFunction_MathIntToName(PaddedSize=0, Number=@Add.Result) {is=struct RigVMUnitNode, template="Int to Name::Execute(in Number,in PaddedSize,out Result)"}
   n1 = RigVMVariableNode() {variable=Shape Settings}
   Get Control Color From Metadata = RigVMFunctionReferenceNode(Color Override=@n1.Value.Color)
   > Control Stack at Item = RigVMFunctionReferenceNode(
    Add Bottom Null=False
    Add Null=true
    Add Secondary Control=False
    Secondary Settings="(Shape=(bVisible=True,Name=\"Default\",Color=(R=1,G=0,B=0,A=1),Transform=(Rotation=(X=0,Y=0,Z=0,W=1),Translation=(X=0,Y=0,Z=0),Scale3D=(X=1,Y=1,Z=1))),Proxy=(bIsProxy=False,ShapeVisibility=BasedOnSelection),DisplayName=\"None\",InitialSpace=LocalSpace,bUsePref ...(truncated)"
    Settings="(Shape=(bVisible=True,Name=\"Square_Thick\",Color=(R=1,G=0,B=0,A=1),Transform=(Rotation=(X=0,Y=0,Z=0,W=1),Translation=(X=0,Y=0,Z=0),Scale3D=(X=0.2,Y=0.75,Z=0.2))),Proxy=(bIsProxy=False,ShapeVisibility=BasedOnSelection),DisplayName=\"None\",InitialSpace=LocalSp ...(truncated)"
    Name=/Script/RigVM.RigVMFunction_NameConcat(A="Upper Twist ", B=@"Int to Name".Result) {is=struct RigVMUnitNode, template="Concat::Execute(in A,in B,out Result)"}.Result
    Settings.Shape.Color=@"Get Control Color From Metadata".Result
    Settings.Shape.Transform=RigVMDispatchNode(
     True=&RigVMFunction_MathTransformMul(B=(Rotation=(X=0,Y=0,Z=0,W=1),Translation=(X=0,Y=0,Z=0),Scale3D=(X=-1,Y=1,Z=1)), A=@n1.Value.Transform) {template="Multiply::Execute(in A,in B,out Result)"}.Result
     False=@n1.Value.Transform
     Condition=RigVMVariableNode() {variable=Negative Side}.Value
    ) {template="DISPATCH_RigVMDispatch_If(in Condition,in True,in False,out Result)"}.Result
    Settings.Shape.Name=@n1.Value.Name
    Settings.Shape.bVisible=@n1.Value.bVisible
    Parent=RigVMVariableNode() {variable=Start Bone}.Value
    Item=@For_Each_2.Element
   )
   > RigVMDispatchNode(Element=@"Control Stack at Item".Control, Array=RigVMVariableNode() {variable=Upper Twist Controls}.Value) {template="DISPATCH_RigVMDispatch_ArrayAdd(io Array,in Element,out Index)"}
   RigVMDispatchNode(Array=RigVMVariableNode() {variable=Lower Twist Controls}.Value, ExecuteContext=@RigVMFunction_Sequence.B) {template="DISPATCH_RigVMDispatch_ArrayReset(io Array)"}
   > For_Each_2_1 = RigVMDispatchNode(Array=RigVMVariableNode() {variable=Lower Twist Bones}.Value) {template="DISPATCH_RigVMDispatch_ArrayIterator(in Array,out Element,out Index,out Count,out Ratio)"}
   Int to Name_1 = &RigVMFunction_MathIntToName(
    PaddedSize=0
    Number=&RigVMFunction_MathIntAdd(B=1, A=@For_Each_2_1.Index) {template="Add::Execute(in A,in B,out Result)"}.Result
   ) {template="Int to Name::Execute(in Number,in PaddedSize,out Result)"}
   n2 = RigVMVariableNode() {variable=Shape Settings}
   Get Control Color From Metadata_1 = RigVMFunctionReferenceNode(Color Override=@n2.Value.Color)
   > Control Stack at Item_1 = RigVMFunctionReferenceNode(
    Add Bottom Null=False
    Add Null=true
    Add Secondary Control=False
    Secondary Settings="(Shape=(bVisible=True,Name=\"Default\",Color=(R=1,G=0,B=0,A=1),Transform=(Rotation=(X=0,Y=0,Z=0,W=1),Translation=(X=0,Y=0,Z=0),Scale3D=(X=1,Y=1,Z=1))),Proxy=(bIsProxy=False,ShapeVisibility=BasedOnSelection),DisplayName=\"None\",InitialSpace=LocalSpace,bUsePref ...(truncated)"
    Settings="(Shape=(bVisible=True,Name=\"Square_Thick\",Color=(R=1,G=0,B=0,A=1),Transform=(Rotation=(X=0,Y=0,Z=0,W=1),Translation=(X=0,Y=0,Z=0),Scale3D=(X=0.2,Y=0.75,Z=0.2))),Proxy=(bIsProxy=False,ShapeVisibility=BasedOnSelection),DisplayName=\"None\",InitialSpace=LocalSp ...(truncated)"
    Name=&RigVMFunction_NameConcat(A="Lower Twist ", B=@"Int to Name_1".Result) {template="Concat::Execute(in A,in B,out Result)"}.Result
    Settings.Shape.Color=@"Get Control Color From Metadata_1".Result
    Settings.Shape.Transform=RigVMDispatchNode(
     True=&RigVMFunction_MathTransformMul(B=(Rotation=(X=0,Y=0,Z=0,W=1),Translation=(X=0,Y=0,Z=0),Scale3D=(X=-1,Y=1,Z=1)), A=@n2.Value.Transform) {template="Multiply::Execute(in A,in B,out Result)"}.Result
     False=@n2.Value.Transform
     Condition=RigVMVariableNode() {variable=Negative Side}.Value
    ) {template="DISPATCH_RigVMDispatch_If(in Condition,in True,in False,out Result)"}.Result
    Parent=RigVMVariableNode() {variable=Mid Bone}.Value
    Settings.Shape.Name=@n2.Value.Name
    Settings.Shape.bVisible=@n2.Value.bVisible
    Item=@For_Each_2_1.Element
   )
   > RigVMDispatchNode(Element=@"Control Stack at Item_1".Control, Array=RigVMVariableNode() {variable=Lower Twist Controls}.Value) {template="DISPATCH_RigVMDispatch_ArrayAdd(io Array,in Element,out Index)"}
   Clone = RigVMDispatchNode(Array=RigVMVariableNode() {variable=Upper Twist Controls}.Value) {template="DISPATCH_RigVMDispatch_ArrayClone(in Array,out Clone)"}
   n3 = RigVMDispatchNode(Array=@Clone.Clone, Other=RigVMVariableNode() {variable=Lower Twist Controls}.Value, ExecuteContext=@RigVMFunction_Sequence.C) {template="DISPATCH_RigVMDispatch_ArrayUnion(io Array,in Other)"}
   Make_Array = RigVMDispatchNode(
    Values=((Type=None,Name="None"),((),Type=None,Name="None"),((),Type=None,Name="None"))
    Values.0=RigVMVariableNode() {variable=Start Bone}.Value
    Values.1=RigVMVariableNode() {variable=Mid Bone}.Value
    Values.2=RigVMVariableNode() {variable=End Bone}.Value
   ) {template="DISPATCH_RigVMDispatch_ArrayMake(in Values,out Array)"}
   > Get Chain Length = RigVMFunctionReferenceNode(Initial=true, Items=@Make_Array.Array)
   Get_Module_Metadata_1 = RigVMDispatchNode(Default=1, Name=Global Control Scale, NameSpace=Root) {template="DISPATCH_RigDispatch_GetModuleMetadata(in Name,in NameSpace,in Default,out Value,out Found)"}
   Num = RigVMDispatchNode(Array=@Make_Array.Array) {template="DISPATCH_RigVMDispatch_ArrayGetNum(in Array,out Num)"}
   > Set Control Scale_2 = RigVMFunctionReferenceNode(
    Global Scale=@Get_Module_Metadata_1.Value
    Auto Scale Factor=RigVMDispatchNode(
     False=1
     True=&RigVMFunction_MathDoubleDiv(B=55, A=@"Get Chain Length".Length).Result
     Condition=/Script/RigVM.RigVMFunction_MathIntGreater(B=1, A=@Num.Num) {is=struct RigVMUnitNode, template="Greater::Execute(in A,in B,out Result)"}.Result
    ) {template="DISPATCH_RigVMDispatch_If(in Condition,in True,in False,out Result)"}.Result
    Scale=RigVMVariableNode() {variable=Control Scale}.Value
    Controls=@n3.Array
   )
   Get Children by Contained Strings = RigVMFunctionReferenceNode(
    Recursive=false
    Seperator=|
    Type to Search=Bone
    Search String=RigVMVariableNode() {variable=Up Twist Search String}.Value
    Item=RigVMVariableNode() {variable=Start Bone}.Value
    ExecuteContext=@RigVMFunction_Sequence_1.B
   )
   > RigVMVariableNode(Value=@"Get Children by Contained Strings".Results) {variable=Upper Twist Bones}
   > Get Children by Contained Strings_1 = RigVMFunctionReferenceNode(
    Recursive=false
    Seperator=|
    Type to Search=Bone
    Search String=RigVMVariableNode() {variable=Lo Twist Search String}.Value
    Item=RigVMVariableNode() {variable=Mid Bone}.Value
   )
   > Reverse = RigVMDispatchNode(Array=@"Get Children by Contained Strings_1".Results) {template="DISPATCH_RigVMDispatch_ArrayReverse(io Array)"}
   > RigVMVariableNode(Value=@Reverse.Array) {variable=Lower Twist Bones}
   RigVMVariableNode(Value=(Type=Connector,Name="Start"), ExecuteContext=@RigVMFunction_Sequence_1.A) {variable=Start Bone}
   > RigVMVariableNode(Value=(Type=Connector,Name="Mid")) {variable=Mid Bone}
   > RigVMVariableNode(Value=(Type=Connector,Name="End")) {variable=End Bone}
   > RigVMVariableNode(Value=(Type=Connector,Name="Parent")) {variable=Parent}
  ]}
 RigVMFunctionLibrary {links [], nodes []}
 RigVMFunction_Sequence_1_1_ContainedGraph {nodes [
   n1 = RigVMFunctionEntryNode()
   n2 = &RigVMFunction_Sequence(ExecuteContext=@n1.ExecuteContext)
   n3 = &RigVMFunction_Sequence(ExecuteContext=@n2.B)
   n4 = &RigVMFunction_Sequence(ExecuteContext=@n3.B)
   RigVMFunctionReturnNode(A=@n2.A, B=@n3.A, C=@n4.A, D=@n4.B)
  ]}
 RigVMFunction_Sequence_1_1_ContainedGraph {nodes [
   n1 = RigVMFunctionEntryNode()
   n2 = &RigVMFunction_Sequence(ExecuteContext=@n1.ExecuteContext)
   n3 = &RigVMFunction_Sequence(ExecuteContext=@n2.B)
   RigVMFunctionReturnNode(A=@n2.A, B=@n3.A, C=@n3.B)
  ]}
 RigVMFunction_Sequence_2_ContainedGraph {nodes [
   n1 = RigVMFunctionEntryNode()
   RigVMFunction_Sequence = &RigVMFunction_Sequence(ExecuteContext=@n1.ExecuteContext)
   n2 = &RigVMFunction_Sequence(ExecuteContext=@RigVMFunction_Sequence.B)
   RigVMFunctionReturnNode(A=@RigVMFunction_Sequence.A, B=@n2.A, C=@n2.B)
  ]}
 RigVMModel {nodes [
   RigUnit_BeginExecution = &RigUnit_BeginExecution()
   RigVMFunction_Sequence_1 = RigVMAggregateNode(ExecuteContext=@RigUnit_BeginExecution.ExecutePin)
   Get Array Parents_1 = RigVMFunctionReferenceNode(ExecuteContext=@RigVMFunction_Sequence_1.B, Items=RigVMVariableNode() {variable=Lower Twist Controls}.Value)
   > Blend Twist_1 = RigVMFunctionReferenceNode(
    Drivens=RigVMDispatchNode(
     False=RigVMVariableNode() {variable=Lower Twist Bones}.Value
     Condition=RigVMVariableNode() {variable=Twist Control}.Value
     True=@"Get Array Parents_1".Parents
    ) {template="DISPATCH_RigVMDispatch_If(in Condition,in True,in False,out Result)"}.Result
    TwistAxis=RigVMVariableNode() {variable=Twist Axis}.Value
    Driver=RigVMVariableNode() {variable=End Bone}.Value
    Debug Axis=RigVMVariableNode() {variable=Debug}.Value
    Weights=RigVMVariableNode() {variable=Lo Twist Weights}.Value
    Reverse=RigVMVariableNode() {variable=Lo Twist Reverse}.Value
   )
   Get Array Parents = RigVMFunctionReferenceNode(Items=RigVMVariableNode() {variable=Upper Twist Controls}.Value, ExecuteContext=@RigVMFunction_Sequence_1.A)
   > Blend Twist = RigVMFunctionReferenceNode(
    Drivens=RigVMDispatchNode(
     False=RigVMVariableNode() {variable=Upper Twist Bones}.Value
     Condition=RigVMVariableNode() {variable=Twist Control}.Value
     True=@"Get Array Parents".Parents
    ) {template="DISPATCH_RigVMDispatch_If(in Condition,in True,in False,out Result)"}.Result
    TwistAxis=RigVMVariableNode() {variable=Twist Axis}.Value
    Debug Axis=RigVMVariableNode() {variable=Debug}.Value
    Reverse=RigVMVariableNode() {variable=Up Twist Reverse}.Value
    Weights=RigVMVariableNode() {variable=Up Twist Weights}.Value
    Driver=RigVMVariableNode() {variable=Start Bone}.Value
   )
   n1 = &RigVMFunction_ControlFlowBranch(Condition=RigVMVariableNode() {variable=Twist Control}.Value, ExecuteContext=@RigVMFunction_Sequence_1.C)
   RigVMFunction_Sequence = &RigVMFunction_Sequence(ExecuteContext=@n1.True)
   For_Each_1 = RigVMDispatchNode(ExecuteContext=@RigVMFunction_Sequence.A, Array=RigVMVariableNode() {variable=Upper Twist Controls}.Value) {template="DISPATCH_RigVMDispatch_ArrayIterator(in Array,out Element,out Index,out Count,out Ratio)"}
   > &RigUnit_SetTransform(
    Weight=1
    bInitial=False
    Item=RigVMDispatchNode(Array=RigVMVariableNode() {variable=Upper Twist Bones}.Value, Index=@For_Each_1.Index) {template="DISPATCH_RigVMDispatch_ArrayGetAtIndex(in Array,in Index,out Element)"}.Element
    Value=&RigUnit_GetTransform(bInitial=False, Item=@For_Each_1.Element).Transform
   )
   For_Each_1_1 = RigVMDispatchNode(ExecuteContext=@RigVMFunction_Sequence.B, Array=RigVMVariableNode() {variable=Lower Twist Controls}.Value) {template="DISPATCH_RigVMDispatch_ArrayIterator(in Array,out Element,out Index,out Count,out Ratio)"}
   > &RigUnit_SetTransform(
    Weight=1
    bInitial=False
    Item=RigVMDispatchNode(Array=RigVMVariableNode() {variable=Lower Twist Bones}.Value, Index=@For_Each_1_1.Index) {template="DISPATCH_RigVMDispatch_ArrayGetAtIndex(in Array,in Index,out Element)"}.Element
    Value=&RigUnit_GetTransform(bInitial=False, Item=@For_Each_1_1.Element).Transform
   )
  ]}
 RigVMModel Backwards Solve Graph {nodes [
   n1 = /Script/ControlRig.RigUnit_InverseExecution() {is=struct RigVMUnitNode}
   RigVMFunction_Sequence = &RigVMFunction_Sequence(ExecuteContext=@n1.ExecutePin)
   For_Each = RigVMDispatchNode(ExecuteContext=@RigVMFunction_Sequence.A, Array=RigVMVariableNode() {variable=Upper Twist Bones}.Value) {template="DISPATCH_RigVMDispatch_ArrayIterator(in Array,out Element,out Index,out Count,out Ratio)"}
   n2 = RigVMVariableNode() {variable=Upper Twist Controls}
   > Set Transform_1_1 = /Script/ControlRig.RigUnit_SetTranslation(
    Space=GlobalSpace
    Weight=1
    bInitial=false
    bPropagateToChildren=True
    Value=&RigUnit_GetTransform(bInitial=false, Item=@For_Each.Element).Transform.Translation
    Item=/Script/ControlRig.RigUnit_HierarchyGetParent(
     bDefaultParent=True
     Child=RigVMDispatchNode(Array=@n2.Value, Index=@For_Each.Index) {template="DISPATCH_RigVMDispatch_ArrayGetAtIndex(in Array,in Index,out Element)"}.Element
    ) {is=struct RigVMUnitNode}.Parent
   ) {is=struct RigVMUnitNode, template="Set Transform::Execute(in Item,in Space,in bInitial,in Value,in Weight,in bPropagateToChildren)"}
   For_Each_1 = RigVMDispatchNode(Array=RigVMVariableNode() {variable=Lower Twist Bones}.Value, ExecuteContext=@RigVMFunction_Sequence.B) {template="DISPATCH_RigVMDispatch_ArrayIterator(in Array,out Element,out Index,out Count,out Ratio)"}
   n3 = RigVMVariableNode() {variable=Lower Twist Controls}
   > Set Transform_1 = &RigUnit_SetTranslation(
    Space=GlobalSpace
    Weight=1
    bInitial=false
    bPropagateToChildren=True
    Value=&RigUnit_GetTransform(bInitial=false, Item=@For_Each_1.Element).Transform.Translation
    Item=&RigUnit_HierarchyGetParent(
     bDefaultParent=True
     Child=RigVMDispatchNode(Array=@n3.Value, Index=@For_Each_1.Index) {template="DISPATCH_RigVMDispatch_ArrayGetAtIndex(in Array,in Index,out Element)"}.Element
    ).Parent
   ) {template="Set Transform::Execute(in Item,in Space,in bInitial,in Value,in Weight,in bPropagateToChildren)"}
  ]}
 RigVMModel Connector Event Graph {nodes [
   n1 = /Script/ControlRig.RigUnit_ConnectorExecution() {is=struct RigVMUnitNode}
   RigVMFunction_Sequence = &RigVMFunction_Sequence(ExecuteContext=@n1.ExecuteContext)
   RigVMFunction_ControlFlowBranch = &RigVMFunction_ControlFlowBranch(
    Condition=RigVMDispatchNode(B=(Type=Connector,Name="Parent"), A=/Script/ControlRig.RigUnit_GetCandidates() {is=struct RigVMUnitNode}.Connector) {template="DISPATCH_RigVMDispatch_CoreEquals(in A,in B,out Result)"}.Result
    ExecuteContext=@RigVMFunction_Sequence.A
   )
   /Script/ControlRig.RigUnit_SetDefaultMatch(
    Default=&RigUnit_HierarchyGetParent(
     bDefaultParent=True
     Child=/Script/ControlRig.RigUnit_ResolveConnector(Connector=(Type=Connector,Name="Start"), SkipSocket=False) {is=struct RigVMUnitNode}.Result
    ).Parent
    ExecutePin=@RigVMFunction_ControlFlowBranch.True
   ) {is=struct RigVMUnitNode, template="Set Default Match::Execute(in Default)"}
   n2 = &RigVMFunction_ControlFlowBranch(
    Condition=RigVMDispatchNode(B=(Type=Connector,Name="Mid"), A=&RigUnit_GetCandidates().Connector) {template="DISPATCH_RigVMDispatch_CoreEquals(in A,in B,out Result)"}.Result
    ExecuteContext=@RigVMFunction_Sequence.B
   )
   n3 = &RigUnit_ResolveConnector(Connector=(Type=Connector,Name="Start"), SkipSocket=False)
   &RigUnit_SetDefaultMatch(
    Default=RigVMDispatchNode(
     Index=-1
     Array=/Script/ControlRig.RigUnit_CollectionChildrenArray(TypeToSearch=Bone, bDefaultChildren=True, bIncludeParent=False, bRecursive=false, Parent=@n3.Result) {is=struct RigVMUnitNode}.Items
    ) {template="DISPATCH_RigVMDispatch_ArrayGetAtIndex(in Array,in Index,out Element)"}.Element
    ExecutePin=@n2.True
   ) {template="Set Default Match::Execute(in Default)"}
  ]}
]
hierarchy [
 Start {type=Connector, loc=<0 0 0>, rot=<0 0 0 1>}
 Root {type=Socket, loc=<0 0 0>, rot=<0 0 0 1>}
 Mid {type=Connector, loc=<0 0 0>, rot=<0 0 0 1>}
 Parent {type=Connector, loc=<0 0 0>, rot=<0 0 0 1>}
 Root {loc=<0 0 0>, rot=<0 0 0 1>}
 pelvis {parent=Root, loc=<0 2.2809 95.8968>, rot=<-0.0224 0.7068 0.0224 -0.7068>}
 spine_01 {parent=pelvis, loc=<3.6771 0 0>, rot=<0 0 -0.1258 -0.9921>}
 spine_02 {parent=spine_01, loc=<6.7951 0 0>, rot=<0 0 0.0302 -0.9995>}
 spine_03 {parent=spine_02, loc=<7.2382 0 0>, rot=<0 0 0.0954 -0.9954>}
 spine_04 {parent=spine_03, loc=<8.5239 0 0>, rot=<0 0 0.0512 -0.9987>}
 spine_05 {parent=spine_04, loc=<19.4398 0 0>, rot=<0 0 0.0059 -1>}
 neck_01 {parent=spine_05, loc=<11.8878 0 0>, rot=<0 0 -0.2073 -0.9783>}
 neck_02 {parent=neck_01, loc=<5.1103 0 0>, rot=<0 0 0.0167 -0.9999>}
 head {parent=neck_02, loc=<4.913 0 0>, rot=<0 0 0.1035 -0.9946>}
 clavicle_l {parent=spine_05, loc=<5.5163 -1.3148 -1.4279>, rot=<0.0833 -0.7546 -0.0413 -0.6496>}
 upperarm_l {parent=clavicle_l, loc=<17.8095 0 0>, rot=<-0.0199 -0.3917 -0.0202 -0.9196>}
 lowerarm_l {parent=upperarm_l, loc=<27.7711 0 0>, rot=<0 0 -0.3336 -0.9427>}
 lowerarm_twist_02_l {parent=lowerarm_l, loc=<9.0837 0 0>, rot=<0.0115 0.0149 0.0175 -0.9997>}
 lowerarm_twist_01_l {parent=lowerarm_l, loc=<18.1674 0 0>, rot=<0.0115 0.0149 0.0175 -0.9997>}
 hand_l {parent=lowerarm_l, loc=<27.2511 0 0>, rot=<-0.5576 -0.0017 0.0206 -0.8299>}
 index_metacarpal_l {parent=hand_l, loc=<3.4445 0.3847 -2.3793>, rot=<0.029 -0.0637 -0.0071 -0.9975>}
 index_01_l {parent=index_metacarpal_l, loc=<5.8771 -0.0432 0.2409>, rot=<0 0 0.2026 -0.9793>}
 index_02_l {parent=index_01_l, loc=<4.08 0 0>, rot=<0 0 0.1296 -0.9916>}
 index_03_l {parent=index_02_l, loc=<2.595 0 0>, rot=<0 0 0.109 -0.994>}
 middle_metacarpal_l {parent=hand_l, loc=<3.3758 0.7536 -0.1829>, rot=<-0.0373 0.0004 -0.0203 -0.9991>}
 middle_01_l {parent=middle_metacarpal_l, loc=<6.0982 0 0>, rot=<0 0 0.2721 -0.9623>}
 middle_02_l {parent=middle_01_l, loc=<5.169 0 0>, rot=<0 0 0.1803 -0.9836>}
 middle_03_l {parent=middle_02_l, loc=<2.474 0 0>, rot=<0 0 0.0872 -0.9962>}
 thumb_01_l {parent=hand_l, loc=<1.9924 -1.3566 -2.5815>, rot=<0.5052 -0.3691 -0.067 -0.7772>}
 thumb_02_l {parent=thumb_01_l, loc=<4.378 0 0>, rot=<0.0336 0.0103 0.2019 -0.9788>}
 thumb_03_l {parent=thumb_02_l, loc=<3.086 0 0>, rot=<0 0 0.0872 -0.9962>}
 pinky_metacarpal_l {parent=hand_l, loc=<3.3144 0.3059 2.3911>, rot=<-0.2182 0.1882 0.0583 -0.9558>}
 pinky_01_l {parent=pinky_metacarpal_l, loc=<4.9576 0.1431 -0.1988>, rot=<0.09 -0.017 0.1281 -0.9875>}
 pinky_02_l {parent=pinky_01_l, loc=<3.816 0 0>, rot=<0 0 0.1847 -0.9828>}
 pinky_03_l {parent=pinky_02_l, loc=<2.04 0 0>, rot=<0 0 0.0429 -0.9991>}
 ring_metacarpal_l {parent=hand_l, loc=<3.3743 0.5425 1.0918>, rot=<-0.1166 0.1006 -0.0257 -0.9877>}
 ring_01_l {parent=ring_metacarpal_l, loc=<5.6455 0.0416 -0.0207>, rot=<0.0542 -0.0132 0.2535 -0.9657>}
 ring_02_l {parent=ring_01_l, loc=<4.977 0 0>, rot=<0 0 0.1647 -0.9863>}
 ring_03_l {parent=ring_02_l, loc=<2.265 0 0>, rot=<0 0 0.0799 -0.9968>}
 upperarm_twist_01_l {parent=upperarm_l, loc=<9.257 0 0>, rot=<-0.0001 0.0188 -0.0028 -0.9998>}
 upperarm_twist_02_l {parent=upperarm_l, loc=<18.5141 0 0>, rot=<0 0 0 1>}
 clavicle_r {parent=spine_05, loc=<5.5162 -1.3148 1.4279>, rot=<0.7546 0.0833 -0.6496 0.0413>}
 upperarm_r {parent=clavicle_r, loc=<-17.8096 0 0.0004>, rot=<-0.0199 -0.3917 -0.0202 -0.9196>}
 lowerarm_r {parent=upperarm_r, loc=<-27.7707 0 0>, rot=<0 0 -0.3336 -0.9427>}
 lowerarm_twist_02_r {parent=lowerarm_r, loc=<-9.0837 0 0>, rot=<0.0115 0.0149 0.0175 -0.9997>}
 lowerarm_twist_01_r {parent=lowerarm_r, loc=<-18.1673 0 0>, rot=<0.0115 0.0149 0.0175 -0.9997>}
 hand_r {parent=lowerarm_r, loc=<-27.251 0 0>, rot=<-0.5576 -0.0017 0.0206 -0.8299>}
 pinky_metacarpal_r {parent=hand_r, loc=<-3.3147 -0.3059 -2.3913>, rot=<-0.2182 0.1882 0.0583 -0.9558>}
 pinky_01_r {parent=pinky_metacarpal_r, loc=<-4.9573 -0.1433 0.1989>, rot=<0.09 -0.017 0.1281 -0.9875>}
 pinky_02_r {parent=pinky_01_r, loc=<-3.816 0 0>, rot=<0 0 0.1847 -0.9828>}
 pinky_03_r {parent=pinky_02_r, loc=<-2.04 0 0>, rot=<0 0 0.0429 -0.9991>}
 ring_metacarpal_r {parent=hand_r, loc=<-3.3742 -0.543 -1.0918>, rot=<-0.1166 0.1006 -0.0257 -0.9877>}
 ring_01_r {parent=ring_metacarpal_r, loc=<-5.6457 -0.0414 0.0207>, rot=<0.0542 -0.0132 0.2535 -0.9657>}
 ring_02_r {parent=ring_01_r, loc=<-4.9771 0 0>, rot=<0 0 0.1647 -0.9863>}
 ring_03_r {parent=ring_02_r, loc=<-2.265 -0.0001 0>, rot=<0 0 0.0799 -0.9968>}
 middle_metacarpal_r {parent=hand_r, loc=<-3.3758 -0.754 0.1828>, rot=<-0.0373 0.0004 -0.0203 -0.9991>}
 middle_01_r {parent=middle_metacarpal_r, loc=<-6.0984 0.0001 0>, rot=<0 0 0.2721 -0.9623>}
 middle_02_r {parent=middle_01_r, loc=<-5.169 0.0001 0>, rot=<0 0 0.1803 -0.9836>}
 middle_03_r {parent=middle_02_r, loc=<-2.474 0 0>, rot=<0 0 0.0872 -0.9962>}
 index_metacarpal_r {parent=hand_r, loc=<-3.4445 -0.3852 2.3793>, rot=<0.029 -0.0637 -0.0071 -0.9975>}
 index_01_r {parent=index_metacarpal_r, loc=<-5.8772 0.0434 -0.241>, rot=<0 0 0.2026 -0.9793>}
 index_02_r {parent=index_01_r, loc=<-4.0799 0 0>, rot=<0 0 0.1296 -0.9916>}
 index_03_r {parent=index_02_r, loc=<-2.5951 0 0>, rot=<0 0 0.109 -0.994>}
 thumb_01_r {parent=hand_r, loc=<-1.9928 1.3567 2.5813>, rot=<0.5052 -0.3691 -0.067 -0.7772>}
 thumb_02_r {parent=thumb_01_r, loc=<-4.3778 -0.0004 -0.0001>, rot=<0.0336 0.0103 0.2019 -0.9788>}
 thumb_03_r {parent=thumb_02_r, loc=<-3.086 0 0>, rot=<0 0 0.0872 -0.9962>}
 upperarm_twist_01_r {parent=upperarm_r, loc=<-9.2569 0 -0.0001>, rot=<-0.0001 0.0188 -0.0028 -0.9998>}
 upperarm_twist_02_r {parent=upperarm_r, loc=<-18.5138 0 -0.0002>, rot=<0 0 0 1>}
 thigh_r {parent=pelvis, loc=<-2.3657 0.1195 9.9691>, rot=<0.0295 -0.0724 0.9965 -0.029>}
 calf_r {parent=thigh_r, loc=<43.3413 0 0>, rot=<0 0 -0.0437 -0.999>}
 foot_r {parent=calf_r, loc=<42.2179 0 0>, rot=<-0.0007 -0.0269 0.0232 -0.9994>}
 ball_r {parent=foot_r, loc=<7.0094 -15.2376 -0.5389>, rot=<0 0 -0.7071 -0.7071>}
 calf_twist_02_r {parent=calf_r, loc=<14.0726 0 0.0377>, rot=<-0.0001 -0.0023 0.0232 -0.9997>}
 calf_twist_01_r {parent=calf_r, loc=<28.1453 0 0.0755>, rot=<-0.0001 -0.0023 0.0232 -0.9997>}
 thigh_twist_01_r {parent=thigh_r, loc=<14.4471 0 0>, rot=<0 -0.0023 -0.0114 -0.9999>}
 thigh_twist_02_r {parent=thigh_r, loc=<28.8942 0 0>, rot=<0 -0.0023 -0.0114 -0.9999>}
 thigh_l {parent=pelvis, loc=<-2.3657 0.11 -9.9692>, rot=<0.0724 0.0295 -0.029 -0.9965>}
 calf_l {parent=thigh_l, loc=<-43.3413 0 0>, rot=<0 0 -0.0437 -0.999>}
 foot_l {parent=calf_l, loc=<-42.2179 0 0>, rot=<-0.0007 -0.0269 0.0232 -0.9994>}
 ball_l {parent=foot_l, loc=<-7.0094 15.2376 0.5389>, rot=<0 0 -0.7071 -0.7071>}
 calf_twist_02_l {parent=calf_l, loc=<-14.0726 0 0>, rot=<-0.0001 -0.0023 0.0232 -0.9997>}
 calf_twist_01_l {parent=calf_l, loc=<-28.1453 0 -0.0754>, rot=<-0.0001 -0.0023 0.0232 -0.9997>}
 thigh_twist_01_l {parent=thigh_l, loc=<-14.4471 0 0>, rot=<0 -0.0023 -0.0114 -0.9999>}
 thigh_twist_02_l {parent=thigh_l, loc=<-28.8942 0 0>, rot=<0 -0.0023 -0.0114 -0.9999>}
 ik_foot_root {parent=Root, loc=<0 0 0>, rot=<0 0 0 1>}
 ik_foot_l {parent=ik_foot_root, loc=<14.0877 -0.9943 8.2376>, rot=<0.0636 0.7042 0.0414 -0.7059>}
 ik_foot_r {parent=ik_foot_root, loc=<-14.0876 -0.9848 8.2382>, rot=<0.7059 0.0414 -0.7042 0.0636>}
 ik_hand_root {parent=Root, loc=<0 0 0>, rot=<0 0 0 1>}
 ik_hand_gun {parent=ik_hand_root, loc=<-47.7683 15.6988 104.4664>, rot=<0.5985 0.0096 0.5206 -0.6089>}
 ik_hand_l {parent=ik_hand_gun, loc=<43.7336 61.6698 58.4081>, rot=<-0.4578 -0.6455 -0.6114 0>}
 ik_hand_r {parent=ik_hand_gun, loc=<0 0 0>, rot=<0 0 0 1>}
 interaction {parent=Root, loc=<0 0 0>, rot=<0 0 0 1>}
 center_of_mass {parent=Root, loc=<0 0 0>, rot=<0 0 0 1>}
 End {type=Connector, loc=<0 0 0>, rot=<0 0 0 1>}
]
variables [
 Control Scale {type=double, default=1}
 Debug {type=bool, default=False}
 End Bone {type=FRigElementKey, default=(Type=None,Name="")}
 Lo Twist Reverse {type=bool, default=True}
 Lo Twist Search String {type=FString, default=lowerarm|twist}
 Lo Twist Weights {type="TArray<double>", default=(0.75,0.25)}
 Lower Twist Bones {type="TArray<FRigElementKey>", default=(())}
 Lower Twist Controls {type="TArray<FRigElementKey>"}
 Mid Bone {type=FRigElementKey, default=(Type=None,Name="")}
 Negative Side {type=bool, default=False}
 Parent {type=FRigElementKey, default=(Type=None,Name="")}
 Shape Settings {
  type=FRigUnit_HierarchyAddControl_ShapeSettings
  default=(bVisible=True,Name="Box_Solid",Color=(R=1,G=1,B=1,A=1),Transform=(Rotation=(X=0,Y=0,Z=0,W=1),Translation=(X=0,Y=0,Z=0),Scale3D=(X=0.05,Y=0.35,Z=0.05)))
 }
 Start Bone {type=FRigElementKey, default=(Type=None,Name="")}
 Twist Axis {type=FVector, default=(X=1,Y=0,Z=0)}
 Twist Control {type=bool, default=True}
 Up Twist Reverse {type=bool, default=False}
 Up Twist Search String {type=FString, default=upperarm|twist}
 Up Twist Weights {type="TArray<double>", default=(0.9,0.25)}
 Upper Twist Bones {type="TArray<FRigElementKey>", default=(())}
 Upper Twist Controls {type="TArray<FRigElementKey>", default=(())}
]
