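# Modular Control Rig module that distributes twist along a biped limb segment between Start and End connectors.
# Construction finds twist bones by name under Start, derives translate weights from distance to End, and optionally spawns offset nulls with controls.
# Forward solve slerps the driver's twist by weight, reversible to use End, optionally blends positions, then copies controls to bones.
# Backwards solve snaps controls onto bones.
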ControlRigBlueprint /EpicControlRig/Modules/Biped/LimbTwist/CRM_Epic_LimbTwist_v02
graphs [
 Add_1_ContainedGraph {nodes [
   n1 = RigVMFunctionEntryNode()
   Add = &RigVMFunction_MathDoubleAdd(A=@n1.A, B=@n1.B)
   RigVMFunctionReturnNode(Result=&RigVMFunction_MathDoubleAdd(A=@Add.Result, B=@n1.C).Result)
  ]}
 ConstructionGraph {nodes [
   RigUnit_PrepareForExecution = &RigUnit_PrepareForExecution()
   RigVMFunction_Sequence_1 = RigVMAggregateNode(ExecuteContext=@RigUnit_PrepareForExecution.ExecutePin)
   RigVMFunction_ControlFlowBranch = &RigVMFunction_ControlFlowBranch(ExecuteContext=@RigVMFunction_Sequence_1.D, Condition=RigVMVariableNode() {variable=Twist Control}.Value)
   RigVMFunction_Sequence = &RigVMFunction_Sequence(ExecuteContext=@RigVMFunction_ControlFlowBranch.True)
   n1 = &RigUnit_HierarchyAddNull(
    Name=Twist Parent Null
    Space=GlobalSpace
    Transform=&RigUnit_GetTransform(bInitial=true, Item=RigVMVariableNode() {variable=Start Bone}.Value).Transform
    ExecutePin=@RigVMFunction_Sequence.A
    Parent=RigVMVariableNode() {variable=Parent}.Value
   )
   > RigVMVariableNode(Value=@n1.Item) {variable=Twist Null}
   > RigVMDispatchNode(Array=RigVMVariableNode() {variable=Twist Controls}.Value) {template="DISPATCH_RigVMDispatch_ArrayReset(io Array)"}
   > For_Each_2 = RigVMDispatchNode(Array=RigVMVariableNode() {variable=Twist Bones}.Value) {template="DISPATCH_RigVMDispatch_ArrayIterator(in Array,out Element,out Index,out Count,out Ratio)"}
   Add = /Script/RigVM.RigVMFunction_MathIntAdd(B=1, A=@For_Each_2.Index) {is=struct RigVMUnitNode, template="Add::Execute(in A,in B,out Result)"}
   Int to Name = /Script/RigVM.RigVMFunction_MathIntToName(PaddedSize=0, Number=@Add.Result) {is=struct RigVMUnitNode, template="Int to Name::Execute(in Number,in PaddedSize,out Result)"}
   n2 = /Script/RigVM.RigVMFunction_NameConcat(A="Twist ", B=@"Int to Name".Result) {is=struct RigVMUnitNode, template="Concat::Execute(in A,in B,out Result)"}
   Concat = &RigVMFunction_NameConcat(B=" Null", A=@n2.Result) {template="Concat::Execute(in A,in B,out Result)"}
   > n3 = &RigUnit_HierarchyAddNull(
    Space=GlobalSpace
    Name=@Concat.Result
    Transform=&RigUnit_GetTransform(bInitial=true, Item=@For_Each_2.Element).Transform
    Parent=RigVMVariableNode() {variable=Twist Null}.Value
   )
   n4 = RigVMVariableNode() {variable=Shape Settings}
   Get Control Color From Metadata = RigVMFunctionReferenceNode(Color Override=@n4.Value.Color)
   Get Display Name = RigVMFunctionReferenceNode(Array=RigVMVariableNode() {variable=Display Names}.Value, Index=@For_Each_2.Index)
   > SpawnControl = /Script/ControlRig.RigUnit_HierarchyAddControlTransform(
    InitialValue=(Rotation=(X=0,Y=0,Z=0,W=1),Translation=(X=0,Y=0,Z=0),Scale3D=(X=1,Y=1,Z=1))
    OffsetSpace=LocalSpace
    OffsetTransform=(Rotation=(X=0,Y=0,Z=0,W=1),Translation=(X=0,Y=0,Z=0),Scale3D=(X=1,Y=1,Z=1))
    Settings="(InitialSpace=LocalSpace,bUsePreferredRotationOrder=False,PreferredRotationOrder=YZX,Limits=(LimitTranslationX=(bMinimum=False,bMaximum=False),LimitTranslationY=(bMinimum=False,bMaximum=False),LimitTranslationZ=(bMinimum=False,bMaximum=False),LimitPitch=(b ...(truncated)"
    Name=@n2.Result
    Settings.Shape.Color=@"Get Control Color From Metadata".Result
    Settings.DisplayName=@"Get Display Name".Display Name
    Parent=@n3.Item
    Settings.Shape.Name=@n4.Value.Name
    Settings.Shape.Transform=@n4.Value.Transform
    Settings.Shape.bVisible=@n4.Value.bVisible
   ) {
    is=struct RigVMUnitNode
    template="SpawnControl::Execute(in InitialValue,in Settings,in OffsetTransform,in OffsetSpace,in Parent,in Name,out Item)"
   }
   > RigVMDispatchNode(Element=@SpawnControl.Item, Array=RigVMVariableNode() {variable=Twist Controls}.Value) {template="DISPATCH_RigVMDispatch_ArrayAdd(io Array,in Element,out Index)"}
   > Set Mirror Behavior = RigVMFunctionReferenceNode(Value=RigVMVariableNode() {variable=FK Mirror Behavior}.Value, Item=@SpawnControl.Item)
   Make_Array = RigVMDispatchNode(
    Values=((Type=None,Name="None"),((),Type=None,Name="None"))
    Values.0=RigVMVariableNode() {variable=Start Bone}.Value
    Values.1=RigVMVariableNode() {variable=End Bone}.Value
   ) {template="DISPATCH_RigVMDispatch_ArrayMake(in Values,out Array)"}
   Get Chain Length = RigVMFunctionReferenceNode(Initial=true, Items=@Make_Array.Array, ExecuteContext=@RigVMFunction_Sequence.B)
   Get_Module_Metadata_1 = RigVMDispatchNode(Default=1, Name=Global Control Scale, NameSpace=Root) {template="DISPATCH_RigDispatch_GetModuleMetadata(in Name,in NameSpace,in Default,out Value,out Found)"}
   Num = RigVMDispatchNode(Array=@Make_Array.Array) {template="DISPATCH_RigVMDispatch_ArrayGetNum(in Array,out Num)"}
   > Set Control Scale_2 = RigVMFunctionReferenceNode(
    Global Scale=@Get_Module_Metadata_1.Value
    Auto Scale Factor=RigVMDispatchNode(
     False=1
     True=&RigVMFunction_MathDoubleDiv(B=20, A=@"Get Chain Length".Length).Result
     Condition=/Script/RigVM.RigVMFunction_MathIntGreater(B=1, A=@Num.Num) {is=struct RigVMUnitNode, template="Greater::Execute(in A,in B,out Result)"}.Result
    ) {template="DISPATCH_RigVMDispatch_If(in Condition,in True,in False,out Result)"}.Result
    Scale=RigVMVariableNode() {variable=Control Scale}.Value
    Controls=RigVMVariableNode() {variable=Twist Controls}.Value
   )
   Get Children by Contained Strings = RigVMFunctionReferenceNode(
    Recursive=false
    Seperator=|
    Type to Search=Bone
    Search String=RigVMVariableNode() {variable=Twist Search String}.Value
    Item=RigVMVariableNode() {variable=Start Bone}.Value
    ExecuteContext=@RigVMFunction_Sequence_1.B
   )
   > RigVMVariableNode(Value=@"Get Children by Contained Strings".Results) {variable=Twist Bones}
   n5 = &RigVMFunction_ControlFlowBranch(Condition=RigVMVariableNode() {variable=Use Translates}.Value, ExecuteContext=@RigVMFunction_Sequence_1.C)
   RigVMDispatchNode(ExecuteContext=@n5.True, Array=RigVMVariableNode() {variable=Translate Weights}.Value) {template="DISPATCH_RigVMDispatch_ArrayReset(io Array)"}
   > Compute Translate Weights = RigVMFunctionReferenceNode(
    Twist Bones=RigVMVariableNode() {variable=Twist Bones}.Value
    Start Bone=RigVMVariableNode() {variable=Start Bone}.Value
    End Bone=RigVMVariableNode() {variable=End Bone}.Value
   )
   > RigVMVariableNode(Value=@"Compute Translate Weights".Translate Weights) {variable=Translate Weights}
   RigVMVariableNode(Value=(Type=Connector,Name="Start"), ExecuteContext=@RigVMFunction_Sequence_1.A) {variable=Start Bone}
   > RigVMVariableNode(Value=(Type=Connector,Name="End")) {variable=End Bone}
   > RigVMVariableNode(Value=(Type=Connector,Name="Parent")) {variable=Parent}
  ]}
 New Function_ContainedGraph {nodes [
   RigVMFunctionEntryNode()
   > RigVMFunction_Sequence_1 = RigVMAggregateNode()
   For_Each_1 = RigVMDispatchNode(ExecuteContext=@RigVMFunction_Sequence_1.B, Array=RigVMVariableNode() {variable=Drivens}.Value) {template="DISPATCH_RigVMDispatch_ArrayIterator(in Array,out Element,out Index,out Count,out Ratio)"}
   n1 = RigVMDispatchNode(Array=RigVMVariableNode() {variable=Weights}.Value, Index=@For_Each_1.Index) {template="DISPATCH_RigVMDispatch_ArrayGetAtIndex(in Array,in Index,out Element)"}
   n2 = RigVMVariableNode() {variable=Twist Output}
   Inverse = /Script/RigVM.RigVMFunction_MathQuaternionInverse(Value=@n2.Value) {is=struct RigVMUnitNode, template="Inverse::Execute(in Value,out Result)"}
   n3 = RigVMDispatchNode(False=@Inverse.Result, Condition=RigVMVariableNode() {variable=Reverse}.Value, True=@n2.Value) {template="DISPATCH_RigVMDispatch_If(in Condition,in True,in False,out Result)"}
   Interpolate = /Script/RigVM.RigVMFunction_MathQuaternionSlerp(B=(X=0,Y=0,Z=0,W=1), T=@n1.Element, A=@n3.Result) {is=struct RigVMUnitNode, template="Interpolate::Execute(in A,in B,in T,out Result)"}
   > Set Transform_2_1 = &RigUnit_SetRotation(
    Weight=1
    bInitial=False
    bPropagateToChildren=true
    Value=RigVMDispatchNode(
     False=@Interpolate.Result
     True=&RigVMFunction_MathQuaternionSlerp(A=(X=0,Y=0,Z=0,W=1), T=@n1.Element, B=@n3.Result) {template="Interpolate::Execute(in A,in B,in T,out Result)"}.Result
     Condition=RigVMVariableNode() {variable=Reverse}.Value
    ) {template="DISPATCH_RigVMDispatch_If(in Condition,in True,in False,out Result)"}.Result
    Item=@For_Each_1.Element
   )
   > &RigUnit_OffsetTransformForItem(
    OffsetTransform=(Rotation=(X=0,Y=0,Z=0,W=1),Translation=(X=0,Y=0,Z=0),Scale3D=(X=1,Y=1,Z=1))
    Weight=1
    bPropagateToChildren=True
    OffsetTransform.Rotation=&RigUnit_GetTransform(Space=LocalSpace, bInitial=true, Item=@For_Each_1.Element).Transform.Rotation
    Item=@For_Each_1.Element
   )
   Get Node Twist Value = RigVMFunctionReferenceNode(
    Item=RigVMVariableNode() {variable=Driver}.Value
    Axis=RigVMVariableNode() {variable=TwistAxis}.Value
    ExecuteContext=@RigVMFunction_Sequence_1.A
   )
   > RigVMVariableNode(Value=@"Get Node Twist Value".Twist) {variable=Twist Output}
   Draw Axis = RigVMFunctionReferenceNode(
    Scale=10
    Thickness=0.2
    Item=RigVMVariableNode() {variable=Drivens}.Value
    Enable=RigVMVariableNode() {variable=Debug Axis}.Value
    ExecuteContext=@RigVMFunction_Sequence_1.C
   )
   > RigVMFunctionReturnNode(Twist Ouput=RigVMVariableNode() {variable=Twist Output}.Value)
  ]}
 New Function_ContainedGraph {nodes [
   RigVMFunctionEntryNode()
   > RigVMDispatchNode(Array=RigVMVariableNode() {variable=Weights Array}.Value) {template="DISPATCH_RigVMDispatch_ArrayReset(io Array)"}
   > For_Each = RigVMDispatchNode(Array=RigVMVariableNode() {variable=Twist Bones}.Value) {template="DISPATCH_RigVMDispatch_ArrayIterator(in Array,out Element,out Index,out Count,out Ratio)"}
   n1 = &RigUnit_GetTransform(bInitial=true, Item=RigVMVariableNode() {variable=End Bone}.Value)
   RigVMFunction_MathVectorDistance = &RigVMFunction_MathVectorDistance(A=&RigUnit_GetTransform(bInitial=true, Item=@For_Each.Element).Transform.Translation, B=@n1.Transform.Translation)
   n2 = RigVMVariableNode() {variable=Start Bone}
   Divide = &RigVMFunction_MathDoubleDiv(
    A=@RigVMFunction_MathVectorDistance.Result
    B=&RigVMFunction_MathVectorDistance(A=&RigUnit_GetTransform(bInitial=true, Item=@n2.Value).Transform.Translation, B=@n1.Transform.Translation).Result
   )
   > RigVMDispatchNode(Element=@Divide.Result, Array=RigVMVariableNode() {variable=Weights Array}.Value) {template="DISPATCH_RigVMDispatch_ArrayAdd(io Array,in Element,out Index)"}
   RigVMFunctionReturnNode(ExecuteContext=@For_Each.Completed, Translate Weights=RigVMVariableNode() {variable=Weights Array}.Value)
  ]}
 New Function_ContainedGraph {nodes [
   RigVMFunctionEntryNode()
   > For_Each = RigVMDispatchNode(Array=RigVMVariableNode() {variable=Drivens}.Value) {template="DISPATCH_RigVMDispatch_ArrayIterator(in Array,out Element,out Index,out Count,out Ratio)"}
   n1 = RigVMDispatchNode(Index=@For_Each.Index, Array=RigVMVariableNode() {variable=Weights}.Value) {template="DISPATCH_RigVMDispatch_ArrayGetAtIndex(in Array,in Index,out Element)"}
   Subtract = &RigVMFunction_MathDoubleSub(A=1, B=@n1.Element)
   > /Script/ControlRig.RigUnit_PositionConstraintLocalSpaceOffset(
    Filter=(bX=True,bY=True,bZ=True)
    Parents=((Item=(Type=Bone,Name="None"),Weight=1),(Item=(Type=Bone,Name="None"),Weight=1))
    Weight=1
    bMaintainOffset=true
    Parents.0.Weight=@n1.Element
    Child=@For_Each.Element
    Parents.1.Weight=@Subtract.Result
    Parents.0.Item=RigVMVariableNode() {variable=Start Bone}.Value
    Parents.1.Item=RigVMVariableNode() {variable=End Bone}.Value
   ) {is=struct RigVMUnitNode}
   RigVMFunctionReturnNode(ExecuteContext=@For_Each.Completed)
  ]}
 New Function_ContainedGraph {nodes [
   n1 = RigVMFunctionEntryNode()
   n2 = RigVMDispatchNode(Array=@n1.Array) {template="DISPATCH_RigVMDispatch_ArrayGetNum(in Array,out Num)"}
   GreaterEqual = /Script/RigVM.RigVMFunction_MathIntGreaterEqual(B=1, A=@n2.Num) {is=struct RigVMUnitNode, template="GreaterEqual::Execute(in A,in B,out Result)"}
   Less = /Script/RigVM.RigVMFunction_MathIntLess(A=@n1.Index, B=@n2.Num) {is=struct RigVMUnitNode, template="Less::Execute(in A,in B,out Result)"}
   RigVMFunction_MathBoolAnd = &RigVMFunction_MathBoolAnd(B=@GreaterEqual.Result, A=@Less.Result)
   > RigVMFunctionReturnNode(Display Name=RigVMDispatchNode(
     False=None
     True=RigVMDispatchNode(Array=@n1.Array, Index=@n1.Index) {template="DISPATCH_RigVMDispatch_ArrayGetAtIndex(in Array,in Index,out Element)"}.Element
     Condition=@RigVMFunction_MathBoolAnd.Result
    ) {template="DISPATCH_RigVMDispatch_If(in Condition,in True,in False,out Result)"}.Result)
  ]}
 RigVMFunctionLibrary {nodes [
   Blend Translate = RigVMCollapseNode(Drivens=(), End Bone=(Type=None,Name="None"), ExecuteContext=(), Start Bone=(Type=None,Name="None"), Weights=())
   Blend Twist = RigVMCollapseNode(Debug Axis=false, Drivens=(), Driver=(Type=None,Name="None"), ExecuteContext=(), TwistAxis=(X=1,Y=0,Z=0), Weights=())
   Compute Translate Weights = RigVMCollapseNode(End Bone=(Type=None,Name="None"), ExecuteContext=(), Start Bone=(Type=None,Name="None"), Twist Bones=())
   Get Display Name = RigVMCollapseNode(Array=())
  ]}
 RigVMFunction_Sequence_1_1_ContainedGraph {nodes [
   n1 = RigVMFunctionEntryNode()
   n2 = &RigVMFunction_Sequence(ExecuteContext=@n1.ExecuteContext)
   n3 = &RigVMFunction_Sequence(ExecuteContext=@n2.B)
   n4 = &RigVMFunction_Sequence(ExecuteContext=@n3.B)
   RigVMFunctionReturnNode(A=@n2.A, B=@n3.A, C=@n4.A, D=@n4.B)
  ]}
 RigVMFunction_Sequence_1_1_ContainedGraph {nodes [
   n1 = RigVMFunctionEntryNode()
   n2 = &RigVMFunction_Sequence(ExecuteContext=@n1.ExecuteContext)
   n3 = &RigVMFunction_Sequence(ExecuteContext=@n2.B)
   n4 = &RigVMFunction_Sequence(ExecuteContext=@n3.B)
   RigVMFunctionReturnNode(A=@n2.A, B=@n3.A, C=@n4.A, D=@n4.B)
  ]}
 RigVMFunction_Sequence_1_1_ContainedGraph {nodes [
   n1 = RigVMFunctionEntryNode()
   n2 = &RigVMFunction_Sequence(ExecuteContext=@n1.ExecuteContext)
   n3 = &RigVMFunction_Sequence(ExecuteContext=@n2.B)
   RigVMFunctionReturnNode(A=@n2.A, B=@n3.A, C=@n3.B)
  ]}
 RigVMFunction_Sequence_1_ContainedGraph {nodes [
   n1 = RigVMFunctionEntryNode()
   RigVMFunction_Sequence = &RigVMFunction_Sequence(ExecuteContext=@n1.ExecuteContext)
   n2 = &RigVMFunction_Sequence(ExecuteContext=@RigVMFunction_Sequence.B)
   RigVMFunctionReturnNode(A=@RigVMFunction_Sequence.A, B=@n2.A, C=@n2.B)
  ]}
 RigVMModel {nodes [
   RigUnit_BeginExecution = &RigUnit_BeginExecution()
   RigVMFunction_Sequence_1 = RigVMAggregateNode(ExecuteContext=@RigUnit_BeginExecution.ExecutePin)
   Get Array Parents = RigVMFunctionReferenceNode(ExecuteContext=@RigVMFunction_Sequence_1.B, Items=RigVMVariableNode() {variable=Twist Controls}.Value)
   n1 = RigVMDispatchNode(
    False=RigVMVariableNode() {variable=Twist Bones}.Value
    Condition=RigVMVariableNode() {variable=Twist Control}.Value
    True=@"Get Array Parents".Parents
   ) {template="DISPATCH_RigVMDispatch_If(in Condition,in True,in False,out Result)"}
   > Blend Twist_0 = RigVMFunctionReferenceNode(
    Drivens=@n1.Result
    Driver=RigVMDispatchNode(
     Condition=RigVMVariableNode() {variable=Twist Reverse}.Value
     False=RigVMVariableNode() {variable=Start Bone}.Value
     True=RigVMVariableNode() {variable=End Bone}.Value
    ) {template="DISPATCH_RigVMDispatch_If(in Condition,in True,in False,out Result)"}.Result
    TwistAxis=RigVMVariableNode() {variable=Twist Axis}.Value
    Debug Axis=RigVMVariableNode() {variable=Debug}.Value
    Reverse=RigVMVariableNode() {variable=Twist Reverse}.Value
    Weights=RigVMVariableNode() {variable=Twist Weights}.Value
   )
   RigVMFunction_ControlFlowBranch = &RigVMFunction_ControlFlowBranch(ExecuteContext=@"Blend Twist_0".ExecuteContext, Condition=RigVMVariableNode() {variable=Use Translates}.Value)
   Blend Translate = RigVMFunctionReferenceNode(
    ExecuteContext=@RigVMFunction_ControlFlowBranch.True
    Start Bone=RigVMVariableNode() {variable=Start Bone}.Value
    End Bone=RigVMVariableNode() {variable=End Bone}.Value
    Weights=RigVMVariableNode() {variable=Translate Weights}.Value
    Drivens=@n1.Result
   )
   &RigUnit_SetTransform(
    Weight=1
    bInitial=False
    Value=&RigUnit_GetTransform(bInitial=False, Item=RigVMVariableNode() {variable=Start Bone}.Value).Transform
    Item=RigVMVariableNode() {variable=Twist Null}.Value
    ExecutePin=@RigVMFunction_Sequence_1.A
   )
   n2 = &RigVMFunction_ControlFlowBranch(Condition=RigVMVariableNode() {variable=Twist Control}.Value, ExecuteContext=@RigVMFunction_Sequence_1.C)
   For_Each_1 = RigVMDispatchNode(ExecuteContext=@n2.True, Array=RigVMVariableNode() {variable=Twist Controls}.Value) {template="DISPATCH_RigVMDispatch_ArrayIterator(in Array,out Element,out Index,out Count,out Ratio)"}
   > &RigUnit_SetTransform(
    Weight=1
    bInitial=False
    Item=RigVMDispatchNode(Array=RigVMVariableNode() {variable=Twist Bones}.Value, Index=@For_Each_1.Index) {template="DISPATCH_RigVMDispatch_ArrayGetAtIndex(in Array,in Index,out Element)"}.Element
    Value=&RigUnit_GetTransform(bInitial=False, Item=@For_Each_1.Element).Transform
   )
   Get_Module_Metadata = RigVMDispatchNode(Default=true, Name=Sec Controls Visibilty, NameSpace=Parent) {template="DISPATCH_RigDispatch_GetModuleMetadata(in Name,in NameSpace,in Default,out Value,out Found)"}
   > SetControlVisibility = /Script/ControlRig.RigUnit_SetControlVisibility(Item=@For_Each_1.Element, bVisible=@Get_Module_Metadata.Value) {is=struct RigVMUnitNode, template="SetControlVisibility::Execute(in Item,in Pattern,in bVisible)"}
   RigVMFunction_MathBoolNot = &RigVMFunction_MathBoolNot(Value=RigVMVariableNode() {variable=Twist Reverse}.Value)
   n3 = &RigVMFunction_ControlFlowBranch(ExecuteContext=(), Condition=@RigVMFunction_MathBoolNot.Result)
   n4 = RigVMVariableNode() {variable=Twist Axis}
   Add = RigVMAggregateNode(A=@n4.Value.X, B=@n4.Value.Y, C=@n4.Value.Z)
   Less = /Script/RigVM.RigVMFunction_MathDoubleLess(B=0, A=@Add.Result) {is=struct RigVMUnitNode, template="Less::Execute(in A,in B,out Result)"}
   Multiply = /Script/RigVM.RigVMFunction_MathVectorMul(B=(X=180,Y=180,Z=180), A=@n4.Value) {is=struct RigVMUnitNode, template="Multiply::Execute(in A,in B,out Result)"}
   RigVMFunction_MathQuaternionFromEuler = &RigVMFunction_MathQuaternionFromEuler(Euler=@Multiply.Result)
   n5 = RigVMVariableNode() {variable=Start Bone}
   n6 = &RigVMFunction_MathQuaternionSwingTwist(Input=&RigUnit_GetTransform(bInitial=False, Item=@n5.Value).Transform.Rotation, TwistAxis=@n4.Value)
   Set Transform_1_1 = &RigUnit_SetRotation(
    Space=GlobalSpace
    Weight=1
    bInitial=False
    bPropagateToChildren=false
    Value=RigVMDispatchNode(
     Condition=@Less.Result
     True=/Script/RigVM.RigVMFunction_MathQuaternionMul(B=@RigVMFunction_MathQuaternionFromEuler.Result, A=@n6.Swing) {is=struct RigVMUnitNode, template="Multiply::Execute(in A,in B,out Result)"}.Result
     False=@n6.Swing
    ) {template="DISPATCH_RigVMDispatch_If(in Condition,in True,in False,out Result)"}.Result
    ExecutePin=@n3.True
    Item=@n5.Value
   )
   RigVMRerouteNode(Value=@RigVMFunction_Sequence_1.D)
  ]}
 RigVMModel Backwards Solve Graph {nodes [
   n1 = /Script/ControlRig.RigUnit_InverseExecution() {is=struct RigVMUnitNode}
   RigVMFunction_Sequence = RigVMAggregateNode(ExecuteContext=@n1.ExecutePin)
   RigVMFunction_ControlFlowBranch = &RigVMFunction_ControlFlowBranch(ExecuteContext=@RigVMFunction_Sequence.B, Condition=RigVMVariableNode() {variable=Use Translates}.Value)
   Get Array Parents = RigVMFunctionReferenceNode(ExecuteContext=@RigVMFunction_ControlFlowBranch.True, Items=RigVMVariableNode() {variable=Twist Controls}.Value)
   > Blend Twist_0 = RigVMFunctionReferenceNode(
    Drivens=@"Get Array Parents".Parents
    Driver=RigVMDispatchNode(
     Condition=RigVMVariableNode() {variable=Twist Reverse}.Value
     False=RigVMVariableNode() {variable=Start Bone}.Value
     True=RigVMVariableNode() {variable=End Bone}.Value
    ) {template="DISPATCH_RigVMDispatch_If(in Condition,in True,in False,out Result)"}.Result
    TwistAxis=RigVMVariableNode() {variable=Twist Axis}.Value
    Debug Axis=RigVMVariableNode() {variable=Debug}.Value
    Reverse=RigVMVariableNode() {variable=Twist Reverse}.Value
    Weights=RigVMVariableNode() {variable=Twist Weights}.Value
   )
   > Blend Translate = RigVMFunctionReferenceNode(
    Drivens=@"Get Array Parents".Parents
    Start Bone=RigVMVariableNode() {variable=Start Bone}.Value
    End Bone=RigVMVariableNode() {variable=End Bone}.Value
    Weights=RigVMVariableNode() {variable=Translate Weights}.Value
   )
   For_Each = RigVMDispatchNode(ExecuteContext=@RigVMFunction_Sequence.C, Array=RigVMVariableNode() {variable=Twist Bones}.Value) {template="DISPATCH_RigVMDispatch_ArrayIterator(in Array,out Element,out Index,out Count,out Ratio)"}
   > &RigUnit_SetTransform(
    Weight=1
    bInitial=False
    Item=RigVMDispatchNode(Array=RigVMVariableNode() {variable=Twist Controls}.Value, Index=@For_Each.Index) {template="DISPATCH_RigVMDispatch_ArrayGetAtIndex(in Array,in Index,out Element)"}.Element
    Value=&RigUnit_GetTransform(bInitial=false, Item=@For_Each.Element).Transform
   )
   &RigUnit_SetTransform(
    Weight=1
    bInitial=False
    Value=&RigUnit_GetTransform(bInitial=False, Item=RigVMVariableNode() {variable=Start Bone}.Value).Transform
    Item=RigVMVariableNode() {variable=Twist Null}.Value
    ExecutePin=@RigVMFunction_Sequence.A
   )
  ]}
 RigVMModel Connector Event Graph {nodes [
   n1 = /Script/ControlRig.RigUnit_ConnectorExecution() {is=struct RigVMUnitNode}
   n2 = /Script/ControlRig.RigUnit_ResolveConnector(Connector=(Type=Connector,Name="Start"), SkipSocket=False) {is=struct RigVMUnitNode}
   Set Default Match To Connector v01 = RigVMFunctionReferenceNode(Connector=(Type=Connector,Name="Parent"), ExecuteContext=@n1.ExecuteContext, Default Match=@n2.Result)
   > Set Default Match To Connector v01_2 = RigVMFunctionReferenceNode(Connector=(Type=Connector,Name="Start Socket"), Default Match=@n2.Result)
   n3 = &RigUnit_ResolveConnector(Connector=(Type=Connector,Name="Start"), SkipSocket=False)
   > Set Default Match To Connector v01_1 = RigVMFunctionReferenceNode(
    Default Match=(Type=Connector,Name="Clavicle")
    Connector=RigVMDispatchNode(
     Index=1
     Array=/Script/ControlRig.RigUnit_CollectionChildrenArray(TypeToSearch=All, bDefaultChildren=True, bIncludeParent=False, bRecursive=False, Parent=@n3.Result) {is=struct RigVMUnitNode}.Items
    ) {template="DISPATCH_RigVMDispatch_ArrayGetAtIndex(in Array,in Index,out Element)"}.Element
   )
  ]}
]
hierarchy [
 Start {type=Connector, loc=<0 0 0>, rot=<0 0 0 1>}
 Root {type=Socket, loc=<0 0 0>, rot=<0 0 0 1>}
 Parent {type=Connector, loc=<0 0 0>, rot=<0 0 0 1>}
 End {type=Connector, loc=<0 0 0>, rot=<0 0 0 1>}
]
variables [
 Control Scale {type=double, default=1}
 Debug {type=bool, default=False}
 Display Names {type="TArray<FName>"}
 End Bone {type=FRigElementKey, default=(Type=None,Name="")}
 FK Mirror Behavior {type=bool, default=False}
 Mid Bone {type=FRigElementKey, default=(Type=None,Name="")}
 Parent {type=FRigElementKey, default=(Type=None,Name="")}
 Shape Settings {
  type=FRigUnit_HierarchyAddControl_ShapeSettings
  default=(bVisible=True,Name="Default",Color=(R=1,G=1,B=1,A=1),Transform=(Rotation=(X=0,Y=0,Z=0,W=1),Translation=(X=0,Y=0,Z=6),Scale3D=(X=0.2,Y=0.2,Z=0.2)))
 }
 Start Bone {type=FRigElementKey, default=(Type=None,Name="")}
 Start Twist Socket {type=FRigElementKey, default=(Type=None,Name="")}
 Translate Weights {type="TArray<double>"}
 Twist Axis {type=FVector, default=(X=1,Y=0,Z=0)}
 Twist Bones {type="TArray<FRigElementKey>", default=(())}
 Twist Control {type=bool, default=True}
 Twist Controls {type="TArray<FRigElementKey>", default=(())}
 Twist Null {type=FRigElementKey, default=(Type=None,Name="")}
 Twist Reverse {type=bool, default=False}
 Twist Search String {type=FString, default=upperarm|twist}
 Twist Weights {type="TArray<double>", default=(0.25,0.75,1)}
 Use Translates {type=bool, default=True}
]
